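# Niagara system for the BoldTouch pickaxe's swing trail, spawning four anim-trail ribbons from bone or socket pairs.
# User.VariantColor switches between magenta and blue color-over-life curves, multiplied by time-of-day curves that brighten them toward both ends of the cycle.
# Ribbon alpha and width fade over life, while a triple-width, translucent black secondary color and width are computed alongside for the two ribbon renderers.
# An older single-variant ScaleColor remains disabled.
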NiagaraSystem /BRCosmetics/Weapons/FORT_Melee/Pickaxe_BoldTouch/FX/NS_Pickaxe_BoldTouch_Trail
dependencies {
 usedDataInterfaces [
  &NiagaraDataInterfaceColorCurve
  &NiagaraDataInterfaceVectorCurve
  &NiagaraDataInterfaceCurve
  &NiagaraDataInterfaceSocketReader
 ]
 usedDynamicInputs []
 usedModules [
  &SystemState
  &self:SystemSpawnScript.SystemScriptSource.SystemScriptGraph.NiagaraNodeAssignment_1.SetVariables_#1
 ]
}
emitters {Trail {
  emitterUpdateScript [
   &EmitterState()
   &FN_SpawnAnimTrail(Number of Anim Trails=4)
  ]
  particleSpawnScript [
   &FN_InitializeAnimTrailBoneIndecies()
   &FN_InitializeAnimTrailBoneIndecies(AnimTrail Selection=1, First Bone or Socket Index=2, Second Bone or Socket Index=3) {module=FN_InitializeAnimTrailBoneIndecies001}
   &FN_InitializeAnimTrailBoneIndecies(AnimTrail Selection=2, First Bone or Socket Index=4, Second Bone or Socket Index=5) {module=FN_InitializeAnimTrailBoneIndecies003}
   &FN_InitializeAnimTrailBoneIndecies(AnimTrail Selection=3, First Bone or Socket Index=6, Second Bone or Socket Index=7) {module=FN_InitializeAnimTrailBoneIndecies002}
   &FN_InitializeAnimTrail(AnimTrailWidth=0.5) {module=FN_InitializeAnimTrail001}
  ]
  particleUpdateScript [
   &ParticleState()
   &FN_UpdateAnimTrail()
   &SolveForcesAndVelocity(Clamp Velocity=false, Limit Acceleration=false)
   &ScaleColor(
    Scale Alpha=&FloatFromCurve(FloatCurve=DataInterface() {propertyValues {Curve {keys [
         {arriveTangent=-3.5298, interpMode=RCIM_Cubic, leaveTangent=-3.5298, tangentMode=RCTM_User, time=0, value=1}
         {interpMode=RCIM_Cubic, tangentMode=RCTM_User, time=1, value=0}
        ]}}})
    Scale RGB=&Multiply_Vector(
     A=&MakeVectorFromLinearColorRGB(Color=&ColorFromCurve(ColorCurve=DataInterface() {propertyValues {
         AlphaCurve {keys [<at 0 1 RCIM_Cubic>,<at 0.9923 0 RCIM_Cubic>]}
         BlueCurve {keys [
           <at 0 2.8634 RCIM_Cubic>
           {arriveTangent=-5.0656, interpMode=RCIM_Cubic, leaveTangent=-5.0656, time=0.2785, value=0.3089}
           <at 0.5425 0.1153>
          ]}
         GreenCurve {keys [
           <at 0 0.1118 RCIM_Cubic>
           {arriveTangent=-0.1606, interpMode=RCIM_Cubic, leaveTangent=-0.1606, time=0.2785, value=0.0031}
           <at 0.5425 0.0247>
          ]}
         RedCurve {keys [<at 0 3 RCIM_Cubic>, {arriveTangent=-5.146, interpMode=RCIM_Cubic, leaveTangent=-5.146, time=0.2785, value=0.6343}, <at 0.5425 0.2082>]}
        }}))
     B=&VectorFromCurve(
      VectorCurve=DataInterface() {propertyValues {
        XCurve {keys [
          {interpMode=RCIM_Cubic, tangentMode=RCTM_Break, time=0, value=2}
          <at 0.4 1>
          {interpMode=RCIM_Cubic, tangentMode=RCTM_Break, time=0.6, value=1}
          {interpMode=RCIM_Cubic, tangentMode=RCTM_Break, time=1, value=2}
         ]}
        YCurve {keys [<at 0 1>,<at 1 1>]}
        ZCurve {keys [
          {interpMode=RCIM_Cubic, tangentMode=RCTM_Break, time=0, value=1.5}
          <at 0.4 1>
          {interpMode=RCIM_Cubic, tangentMode=RCTM_Break, time=0.6, value=1}
          {interpMode=RCIM_Cubic, tangentMode=RCTM_Break, time=1, value=1.5}
         ]}
       }}
      CurveIndex=Linked() {linkedVariable {name=NPC.FortniteNPC.TimeOfDay_0-1, type=NiagaraFloat}}
     )
    )
    ScaleA=true
    ScaleRGB=true
   ) {enabled=false}
   &ScaleColor(
    Scale Alpha=&FloatFromCurve(FloatCurve=DataInterface() {propertyValues {Curve {keys [
         {arriveTangent=-3.5298, interpMode=RCIM_Cubic, leaveTangent=-3.5298, tangentMode=RCTM_User, time=0, value=1}
         {interpMode=RCIM_Cubic, tangentMode=RCTM_User, time=1, value=0}
        ]}}})
    Scale RGB=&MergeHelper_VectorSwitch(
     0=&Multiply_Vector(
      A=&MakeVectorFromLinearColorRGB(Color=&ColorFromCurve(ColorCurve=DataInterface() {propertyValues {
          AlphaCurve {keys [<at 0 1 RCIM_Cubic>,<at 0.9923 0 RCIM_Cubic>]}
          BlueCurve {keys [
            <at 0 2.8634 RCIM_Cubic>
            {arriveTangent=-3.9739, interpMode=RCIM_Cubic, leaveTangent=-3.9739, time=0.2472, value=0.5558}
            <at 0.6682 0.2082>
           ]}
          GreenCurve {keys [
            <at 0 0.1118 RCIM_Cubic>
            {arriveTangent=-0.1423, interpMode=RCIM_Cubic, leaveTangent=-0.1423, time=0.2472, value=0.0561}
            <at 0.6682 0.0167>
           ]}
          RedCurve {keys [<at 0 3 RCIM_Cubic>, {arriveTangent=-4.1859, interpMode=RCIM_Cubic, leaveTangent=-4.1859, time=0.2472, value=0.6343}, <at 0.6682 0.2032>]}
         }}))
      B=&VectorFromCurve(
       VectorCurve=DataInterface() {propertyValues {
         XCurve {keys [
           {interpMode=RCIM_Cubic, tangentMode=RCTM_Break, time=0, value=2}
           <at 0.4 1>
           {interpMode=RCIM_Cubic, tangentMode=RCTM_Break, time=0.6, value=1}
           {interpMode=RCIM_Cubic, tangentMode=RCTM_Break, time=1, value=2}
          ]}
         YCurve {keys [<at 0 1>,<at 1 1>]}
         ZCurve {keys [
           {interpMode=RCIM_Cubic, tangentMode=RCTM_Break, time=0, value=1.5}
           <at 0.4 1>
           {interpMode=RCIM_Cubic, tangentMode=RCTM_Break, time=0.6, value=1}
           {interpMode=RCIM_Cubic, tangentMode=RCTM_Break, time=1, value=1.5}
          ]}
        }}
       CurveIndex=Linked() {linkedVariable {name=NPC.FortniteNPC.TimeOfDay_0-1, type=NiagaraFloat}}
      )
     )
     1=&Multiply_Vector(
      A=&MakeVectorFromLinearColorRGB(Color=&ColorFromCurve(ColorCurve=DataInterface() {propertyValues {
          AlphaCurve {keys [<at 0 1 RCIM_Cubic>,<at 0.9923 0 RCIM_Cubic>]}
          BlueCurve {keys [<at 0.0827 3 RCIM_Cubic>,<at 0.6776 0.9823 RCIM_Cubic>,<at 1 0.279>]}
          GreenCurve {keys [<at 0.0827 1.3369 RCIM_Cubic>,<at 0.6776 0.6525 RCIM_Cubic>,<at 1 0.048>]}
          RedCurve {keys [<at 0.0827 0.1039 RCIM_Cubic>,<at 0.6776 0.3006 RCIM_Cubic>,<at 1 0.0636>]}
         }}))
      B=&VectorFromCurve(
       VectorCurve=DataInterface() {propertyValues {
         XCurve {keys [
           {interpMode=RCIM_Cubic, tangentMode=RCTM_Break, time=0, value=1.5}
           <at 0.4 1>
           {interpMode=RCIM_Cubic, tangentMode=RCTM_Break, time=0.6, value=1}
           {interpMode=RCIM_Cubic, tangentMode=RCTM_Break, time=1, value=1.5}
          ]}
         YCurve {keys [<at 0 1>,<at 1 1>]}
         ZCurve {keys [
           {interpMode=RCIM_Cubic, tangentMode=RCTM_Break, time=0, value=2}
           <at 0.4 1.5>
           {interpMode=RCIM_Cubic, tangentMode=RCTM_Break, time=0.6, value=1.5}
           {interpMode=RCIM_Cubic, tangentMode=RCTM_Break, time=1, value=2}
          ]}
        }}
       CurveIndex=Linked() {linkedVariable {name=NPC.FortniteNPC.TimeOfDay_0-1, type=NiagaraFloat}}
      )
     )
     2=?
     3=?
     Index=&MakeIntFromFloat(Float=$User.VariantColor)
     Num=::NewEnumerator0
    )
    ScaleA=true
    ScaleRGB=true
   ) {module=ScaleColor001}
   &ScaleRibbonWidth(Ribbon Width Scale=&FloatFromCurve(FloatCurve=DataInterface() {propertyValues {Curve {keys [{arriveTangent=-0.0042, interpMode=RCIM_Cubic, leaveTangent=-0.0042, tangentMode=RCTM_User, time=0, value=1}, <at 1 0>]}}}))
   /SetVariables(
    Particles.SecondaryColor:LinearColor=&MultiplyLinearColor(
     Linear Color=<rgba 0 0 0 0.3>
     Scale Factor=&MakeVector4FromVectorAndFloat(
      "Vector (RGB)"=<1 1 1>
      "Float (Alpha)"=&FloatFromCurve(FloatCurve=DataInterface() {propertyValues {Curve {
          defaultValue=1
          keys [{interpMode=RCIM_Cubic, tangentMode=RCTM_User, time=0, value=1}, {interpMode=RCIM_Cubic, tangentMode=RCTM_User, time=0.5, value=0}]
         }}})
     )
    )
    Particles.SecondaryWidth:NiagaraFloat=&Multiply_Float(A=$Particles.RibbonWidth, B=3)
   ) {isSetParameters=true}
  ]
  rendererClasses [&NiagaraRibbonRendererProperties]
  renderers [
   {rendererClass=&NiagaraRibbonRendererProperties}
   {rendererClass=&NiagaraRibbonRendererProperties, rendererIndex=1}
  ]
 }}
summary {
 systemName=NS_Pickaxe_BoldTouch_Trail
 emitters [{emitterName=Trail, rendererClasses [&NiagaraRibbonRendererProperties]}]
 userVariables [
  User.RibbonIDOffset {type=NiagaraInt32}
  User.Skeletal Mesh {
   type=NiagaraDataInterfaceSkeletalMesh
   defaultValue {
    struct=NiagaraExt_VariableValue_DataInterface
    value {dataInterface=&self:NiagaraDataInterfaceSkeletalMesh_2}
   }
  }
  User.VariantColor {type=NiagaraFloat, defaultValue=0, description=""}
 ]
}
