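# Animation Blueprint for the BoldTouch pickaxe mesh, layering bone adjustments for its root and claw fingers onto an input pose.
# The AnimGraph scales the root up, then chains additive bone-space rotation nodes for index, middle, pinkie and thumb.
# Each update computes four phase-offset sine yaw rotators for claw sway, though their bindings to those nodes are not shown.
# Customization loading flags large male characters, with no visible use of that flag.
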AnimBlueprint /BRCosmetics/Weapons/FORT_Melee/Pickaxe_BoldTouch/Meshes/BoldTouch_Axe_AnimBP
target_skeleton=&/BoldTouch_Axe_Skeleton
graphs [
 AnimGraph {nodes [
   AnimGraphNode_SubInput_0 = Input Pose()
   n1 = "Transform (Modify) Bone - Bone: b_AR_Root"(ComponentPose=Local To Component(LocalPose=@AnimGraphNode_SubInput_0), Alpha=1, Translation="0,0,0", Rotation="0,0,0", Scale="1.2,1.2,1.2") {
    scaleMode=BMM_Additive
    scaleSpace=BCS_BoneSpace
    translationMode=BMM_Additive
    translationSpace=BCS_BoneSpace
    alphaBoolBlend {blendOption=Linear}
    boneToModify {boneName=b_AR_Root}
    scale=<1.2 1.2 1.2>
   }
   n2 = "Transform (Modify) Bone - Bone: dyn_index_r_00"(ComponentPose=@n1, Alpha=1, Translation="0,0,0", Rotation="0,0,0", Scale="1,1,1") {rotationMode=BMM_Additive, rotationSpace=BCS_BoneSpace, alphaBoolBlend {blendOption=Linear}, boneToModify {boneName=dyn_index_r_00}}
   n3 = "Transform (Modify) Bone - Bone: dyn_middle_r_00"(ComponentPose=@n2, Alpha=1, Translation="0,0,0", Rotation="0,0,0", Scale="1,1,1") {rotationMode=BMM_Additive, rotationSpace=BCS_BoneSpace, alphaBoolBlend {blendOption=Linear}, boneToModify {boneName=dyn_middle_r_00}}
   n4 = "Transform (Modify) Bone - Bone: dyn_pinkie_r_00"(ComponentPose=@n3, Alpha=1, Translation="0,0,0", Rotation="0,0,0", Scale="1,1,1") {rotationMode=BMM_Additive, rotationSpace=BCS_BoneSpace, alphaBoolBlend {blendOption=Linear}, boneToModify {boneName=dyn_pinkie_r_00}}
   n5 = "Transform (Modify) Bone - Bone: dyn_thumb_r_00"(ComponentPose=@n4, Alpha=1, Translation="0,0,0", Rotation="0,0,0", Scale="1,1,1") {rotationMode=BMM_Additive, rotationSpace=BCS_BoneSpace, alphaBoolBlend {blendOption=Linear}, boneToModify {boneName=dyn_thumb_r_00}}
   Output Pose(Result=Component To Local(ComponentPose=@n5)) {layerGroup=None}
  ]}
 EventGraph {nodes [
   Event BlueprintUpdateAnimation()
   > n1 = Sequence!(execute) -> (then_0, then_1, then_2:Exec, then_3:Exec)
   n2 = GetRealTimeSeconds()
   n3 = float * float(A="Sin (Radians)"(A=float * float(A=@n2, B=1.5)), B=7.5) {is=title K2Node_PromotableOperator}
   > Set ClawRot1(ClawRot1:Rotator=MakeRotator(Yaw=float + float(A=@n3, B=5))) -> (Output_Get:Rotator)
   n4 = float * float(A=GetRealTimeSeconds(), B=1.5)
   n5 = float * float(A="Sin (Radians)"(A=float + float(A=@n4, B=0.5)), B=5)
   Set ClawRot2(execute=@n1.then_1, ClawRot2:Rotator=MakeRotator(Yaw=float + float(A=@n5, B=5))) -> (Output_Get:Rotator)
   n6 = float * float(A=GetRealTimeSeconds(), B=1.5)
   n7 = float * float(A="Sin (Radians)"(A=float + float(A=@n6, B=1)), B=5)
   Set ClawRot3(execute=@n1.then_2, ClawRot3:Rotator=MakeRotator(Yaw=float + float(A=@n7, B=5))) -> (Output_Get:Rotator)
   n8 = float * float(A=GetRealTimeSeconds(), B=1.5)
   n9 = float * float(A="Sin (Radians)"(A=float + float(A=@n8, B=1.5)), B=5)
   Set ClawRot4(execute=@n1.then_3, ClawRot4:Rotator=MakeRotator(Yaw=float + float(A=@n9, B=5))) -> (Output_Get:Rotator)
   n10 = Event OnCharacterAndPickaxeCustomizationLoaded!() -> (
    OutputDelegate
    then
    CustomBaseMesh:Skeletal Mesh Component Object Reference
    CustomParts:Array of Custom Character Part Object References
    CustomMeshParts:Array of Skeletal Mesh Component Budgeted Object References
   ) {is=title K2Node_Event}
   n11 = Get Array Item with EFort Custom Part Type(Array:Array of Custom Character Part Object References=@n10.CustomParts) -> (Output:Custom Character Part Object Reference)
   n12 = Get BodyTypesPermitted!(self:Custom Character Part Object Reference=@n11) -> (BodyTypesPermitted:EFortCustomBodyType Enum=NONE)
   n13 = Get GenderPermitted!(self:Custom Character Part Object Reference=@n11) -> (GenderPermitted:EFortCustomGender Enum=Invalid)
   > Set Large Character(Large Character:bool=AND Boolean(
     A="Equal (Enum)"(A:EFortCustomBodyType Enum=@n12, B:EFortCustomBodyType Enum=Large)
     B="Equal (Enum)"(A:EFortCustomGender Enum=@n13, B:EFortCustomGender Enum=Male)
    )) -> (Output_Get:bool)
   TryGetPawnOwner()
  ]}
]
variables [
 ClawRot1 {type=Rotator, default=<rot 0 -13.8273 0>}
 ClawRot2 {type=Rotator, default=<rot 0 -17.788 0>}
 ClawRot3 {type=Rotator, default=<rot 0 -19.8419 0>}
 ClawRot4 {type=Rotator, default=<rot 0 -19.4862 0>}
 Large Character {type=boolean}
]
